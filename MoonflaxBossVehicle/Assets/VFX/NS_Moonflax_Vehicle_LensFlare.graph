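# Niagara lens flare system for the front of the Moonflax boss vehicle, tinted by vehicle color and spawned only with a driver.
# Its active GPU emitter pins two particles to skeletal mesh sockets each update, rendering them through two sprite renderers.
# Camera-facing dot products and a scratch-module occlusion test drive size and brightness, while time of day limits visibility to evening and night.
# Three variant emitters, one sprite and two mesh, are disabled.
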NiagaraSystem /MoonflaxBossVehicle/Assets/VFX/NS_Moonflax_Vehicle_LensFlare
dependencies {
 usedDataInterfaces [
  /Script/Niagara.NiagaraDataInterfaceOcclusion
  &NiagaraDataInterfaceCamera
  &NiagaraDataInterfaceCurve
  &NiagaraDataInterfaceSkeletalMesh
  &NiagaraDataInterfaceArrayColor
 ]
 usedDynamicInputs [&SelectLinearColorFromArray]
 usedModules [
  &self:SystemSpawnScript.SystemScriptSource.SystemScriptGraph.NiagaraNodeAssignment_1.SetVariables_#1
  &SystemState
  &self:SystemSpawnScript.SystemScriptSource.SystemScriptGraph.NiagaraNodeAssignment_0.SetVariables_#2
 ]
}
emitters {
 Front_LensFlare {
  simTarget=GPUComputeSim
  emitterUpdateScript [
   &EmitterState(Life Cycle Mode=::NewEnumerator1, Loop Duration=1, UseLoopDelay=false)
   &SpawnBurst_Instantaneous(
    Spawn Count=2
    Spawn Probability=&MakeCustomFloatFromBool(Bool=$User.VehicleHasDriver, True Float=1, False Float=0)
    Use Spawn Probability=true
    UseLoopCountLimit=false
   )
  ]
  particleSpawnScript [
   &PartitionParticles()
   &InitializeParticle(
    Color=$System.VehicleColor
    Sprite Rotation Mode=::NewEnumerator0
    Sprite Size=&Multiply_Vector2D_ByFloat(Float=$User.FlareScale, Vector2D=<125 75>)
    Sprite Size Mode=::NewEnumerator4
    UsePositionOffset=false
    Write Material Random=true
   )
  ]
  particleUpdateScript [
   &ParticleState(Kill Particles When Lifetime Has Elapsed=false, Let Infinitely Lived Particles Die When Emitter Deactivates=true)
   &SkeletalMeshLocation(
    Bone / Socket Index=$Particles.PartitionParticles.Partition
    MaskSpawnGroup=false
    Mesh Sampling Type=::NewEnumerator5
    Sampled Position Offset=<25 0 0>
    Skeletal Mesh=$System.SkeletalMesh
    Socket Sampling Mode=::NewEnumerator5
   )
   /SetVariables(
    Particles.Streak_01_Offset:NiagaraPosition=&AddVectorToPosition(Position=$Particles.SkeletalMeshLocation.SampledPosition, Vector=<0 0 0>)
    Particles.Streak_01_Size:Vector2f=<350 25>
   ) {isSetParameters=true, module=SetVariables_#3}
   &self:ScratchModule(
    NumberOfRings=5
    OcclusionSize=4.5
    Position:NiagaraPosition=&TransformPosition(Position=$Particles.Streak_01_Offset, Source Space=::Local, Destination Space=::World)
    SamplesPerRing=5
    SystemX Sample Offset=5
   ) {module=ScratchModule}
   /SetVariables(Particles.CameraScale:NiagaraFloat=&MakeFloatFromVector(
     Vector=&Multiply_VectorByFloat(
      Vector=&ReturnCameraVector(Camera Vectors=::NewEnumerator4, Coordinate Space=::Simulation, Origin Position=$Particles.Position)
      Float=&DotProduct(
       Vector A=<1 0 -1.5>
       Vector B=&ReturnCameraVector(Camera Vectors=::NewEnumerator4, Coordinate Space=::Simulation, Origin Position=$Particles.Position)
      )
     )
     Channel=::NewEnumerator0
    )) {isSetParameters=true, module=SetVariables_#4}
   /SetVariables(Particles.CameraScale:NiagaraFloat=&Multiply_Float(
     A=$Particles.CameraScale
     B=&RemapRange(
      Input Value To Remap=&DotProduct(
       Vector A=&NormalizeVector(
        Vector To Normalize=&RotateVectorByQuaternion(Vector To Rotate=<1 0 0>, Quaternion=$StackContext.SkeletalMeshLocation.SampledRotation)
        Safe Normalize=true
        Safe Fallback Threshold=0
        Safe Fallback Vector=<0 0 0.001>
       )
       Vector B=&NormalizeVector(
        Vector To Normalize=&ReturnCameraVector(Camera Vectors=::NewEnumerator1, Coordinate Space=::Simulation)
        Safe Normalize=true
        Safe Fallback Threshold=0
        Safe Fallback Vector=<0 0 0.001>
       )
      )
      Input Min=0
      Input Max=-1
      Output Min=0
      Output Max=1
     )
    )) {isSetParameters=true}
   &ScaleSpriteSize(
    Uniform Curve Index=$Particles.CameraScale
    Uniform Curve Sprite Scale=DataInterface() {propertyValues {Curve {keys [
        {arriveTangent=4.7849, interpMode=RCIM_Cubic, leaveTangent=4.7849, tangentMode=RCTM_User, time=0, value=0.5}
        {arriveTangent=0.9156, interpMode=RCIM_Cubic, leaveTangent=0.9156, tangentMode=RCTM_User, time=0.3, value=1.275}
        {tangentMode=RCTM_User, time=1, value=1.5}
       ]}}}
   )
   &ScaleColor(Scale Alpha=$Particles.Visibility, Scale RGB=&VectorFromFloat(Value=3), ScaleA=true, ScaleRGB=true) {module=ScaleColor002}
   &ScaleColor(
    Scale Alpha=&FloatFromCurve(
     FloatCurve=DataInterface() {propertyValues {Curve {keys [
         {interpMode=RCIM_Cubic, tangentMode=RCTM_User, time=0, value=0}
         {arriveTangent=0.4636, interpMode=RCIM_Cubic, leaveTangent=0.4636, tangentMode=RCTM_User, time=0.2, value=0.8}
         {arriveTangentWeight=1.2316, interpMode=RCIM_Cubic, tangentMode=RCTM_User, time=1, value=1}
        ]}}}
     CurveIndex=$Particles.CameraScale
    )
    Scale RGB=&Multiply_VectorByFloat(Vector=<1 1 1>, Float=$Particles.CameraScale)
    ScaleA=true
    ScaleRGB=true
   )
   &ScaleColor(
    Scale Alpha=&FN_Select_Float_from_Time_Of_Day(Morning=0, Day=0, Evening=1, Night=1)
    ScaleA=true
    ScaleRGB=false
   ) {module=ScaleColor001}
  ]
  rendererClasses [&NiagaraSpriteRendererProperties]
  renderers [
   {rendererClass=&NiagaraSpriteRendererProperties}
   {rendererClass=&NiagaraSpriteRendererProperties, rendererIndex=1}
  ]
 }
 Front_LensFlare001 {
  bEnabled=false
  simTarget=GPUComputeSim
  emitterUpdateScript [
   &EmitterState(Life Cycle Mode=::NewEnumerator1, Loop Duration=1, UseLoopDelay=false)
   &SpawnBurst_Instantaneous(
    Spawn Count=2
    Spawn Probability=&MakeCustomFloatFromBool(Bool=$User.VehicleHasDriver, True Float=1, False Float=0)
    Use Spawn Probability=true
    UseLoopCountLimit=false
   )
  ]
  particleSpawnScript [
   &PartitionParticles()
   &InitializeParticle(
    Color=$System.VehicleColor
    Sprite Rotation Mode=::NewEnumerator0
    Sprite Size=&Multiply_Vector2D_ByFloat(Float=$User.FlareScale, Vector2D=<125 75>)
    Sprite Size Mode=::NewEnumerator4
    UsePositionOffset=false
    Write Material Random=true
   )
   &SkeletalMeshLocation(
    Bone / Socket Index=$Particles.PartitionParticles.Partition
    MaskSpawnGroup=false
    Mesh Sampling Type=::NewEnumerator5
    Skeletal Mesh=$System.SkeletalMesh
    Socket Sampling Mode=::NewEnumerator5
   )
   /SetVariables(
    Particles.Streak_01_Offset:NiagaraPosition=&AddVectorToPosition(Position=$Particles.SkeletalMeshLocation.SampledPosition, Vector=<0 0 0>)
    Particles.Streak_01_Size:Vector2f=<350 25>
   ) {isSetParameters=true, module=SetVariables_#3}
  ]
  particleUpdateScript [
   &ParticleState(Kill Particles When Lifetime Has Elapsed=false, Let Infinitely Lived Particles Die When Emitter Deactivates=true)
   &self:ScratchModule(
    NumberOfRings=5
    OcclusionSize=4.5
    Position:NiagaraPosition=&TransformPosition(Position=$Particles.Position, Source Space=::Local, Destination Space=::World)
    SamplesPerRing=5
   ) {module=ScratchModule}
   /SetVariables(Particles.CameraScale:NiagaraFloat=&MakeFloatFromVector(
     Vector=&Multiply_VectorByFloat(
      Vector=&ReturnCameraVector(Camera Vectors=::NewEnumerator4, Coordinate Space=::Simulation, Origin Position=$Particles.Position)
      Float=&DotProduct(
       Vector A=<1 0 -1.5>
       Vector B=&ReturnCameraVector(Camera Vectors=::NewEnumerator4, Coordinate Space=::Simulation, Origin Position=$Particles.Position)
      )
     )
     Channel=::NewEnumerator0
    )) {isSetParameters=true, module=SetVariables_#4}
   &ScaleSpriteSize(
    Uniform Curve Index=$Particles.CameraScale
    Uniform Curve Sprite Scale=DataInterface() {propertyValues {Curve {keys [
        {arriveTangent=4.7849, interpMode=RCIM_Cubic, leaveTangent=4.7849, tangentMode=RCTM_User, time=0, value=0.5}
        {arriveTangent=0.9156, interpMode=RCIM_Cubic, leaveTangent=0.9156, tangentMode=RCTM_User, time=0.3, value=1.275}
        {tangentMode=RCTM_User, time=1, value=1.5}
       ]}}}
   )
   &ScaleColor(Scale Alpha=$Particles.Visibility, Scale RGB=&VectorFromFloat(Value=3), ScaleA=true, ScaleRGB=true) {module=ScaleColor002}
   &ScaleColor(
    Scale Alpha=&FloatFromCurve(
     FloatCurve=DataInterface() {propertyValues {Curve {keys [
         {interpMode=RCIM_Cubic, tangentMode=RCTM_User, time=0, value=0}
         {arriveTangent=0.4636, interpMode=RCIM_Cubic, leaveTangent=0.4636, tangentMode=RCTM_User, time=0.2, value=0.8}
         {arriveTangentWeight=1.2316, interpMode=RCIM_Cubic, tangentMode=RCTM_User, time=1, value=1}
        ]}}}
     CurveIndex=$Particles.CameraScale
    )
    Scale RGB=&Multiply_VectorByFloat(Vector=<1 1 1>, Float=$Particles.CameraScale)
    ScaleA=true
    ScaleRGB=true
   )
   &ScaleColor(
    Scale Alpha=&FN_Select_Float_from_Time_Of_Day(Morning=0, Day=0, Evening=1, Night=1)
    ScaleA=true
    ScaleRGB=false
   ) {module=ScaleColor001}
  ]
  rendererClasses [&NiagaraSpriteRendererProperties]
  renderers [
   {rendererClass=&NiagaraSpriteRendererProperties}
   {rendererClass=&NiagaraSpriteRendererProperties, rendererIndex=1}
  ]
 }
 Front_LensFlare002 {
  bEnabled=false
  simTarget=GPUComputeSim
  emitterUpdateScript [
   &EmitterState(Life Cycle Mode=::NewEnumerator1, Loop Duration=1, UseLoopDelay=false)
   &SpawnBurst_Instantaneous(
    Spawn Count=2
    Spawn Probability=&MakeCustomFloatFromBool(Bool=$User.VehicleHasDriver, True Float=1, False Float=0)
    Use Spawn Probability=true
    UseLoopCountLimit=false
   )
  ]
  particleSpawnScript [
   &PartitionParticles()
   &InitializeParticle(
    Color=$System.VehicleColor
    Mesh Scale Mode=::NewEnumerator3
    Mesh Uniform Scale=4
    Sprite Rotation Mode=::NewEnumerator0
    Sprite Size=&Multiply_Vector2D_ByFloat(Float=$User.FlareScale, Vector2D=<125 75>)
    Sprite Size Mode=::NewEnumerator4
    UsePositionOffset=false
    Write Material Random=true
   )
  ]
  particleUpdateScript [
   &ParticleState(Kill Particles When Lifetime Has Elapsed=false, Let Infinitely Lived Particles Die When Emitter Deactivates=true)
   &SkeletalMeshLocation(
    Bone / Socket Index=$Particles.PartitionParticles.Partition
    MaskSpawnGroup=false
    Mesh Sampling Type=::NewEnumerator5
    Sampled Position Offset=<25 0 0>
    Skeletal Mesh=$System.SkeletalMesh
    Socket Sampling Mode=::NewEnumerator5
   )
  ]
  rendererClasses [&NiagaraMeshRendererProperties]
  renderers [{rendererClass=&NiagaraMeshRendererProperties, rendererIndex=0}]
 }
 Front_LensFlare003 {
  bEnabled=false
  simTarget=GPUComputeSim
  emitterUpdateScript [
   &EmitterState(Life Cycle Mode=::NewEnumerator1, Loop Duration=1, UseLoopDelay=false)
   &SpawnBurst_Instantaneous(
    Spawn Count=2
    Spawn Probability=&MakeCustomFloatFromBool(Bool=$User.VehicleHasDriver, True Float=1, False Float=0)
    Use Spawn Probability=true
    UseLoopCountLimit=false
   )
  ]
  particleSpawnScript [
   &PartitionParticles()
   &InitializeParticle(
    Color=$System.VehicleColor
    Mesh Scale Mode=::NewEnumerator3
    Mesh Uniform Scale=15
    Sprite Rotation Mode=::NewEnumerator0
    Sprite Size=&Multiply_Vector2D_ByFloat(Float=$User.FlareScale, Vector2D=<125 75>)
    Sprite Size Mode=::NewEnumerator4
    UsePositionOffset=false
    Write Material Random=true
   )
  ]
  particleUpdateScript [
   &ParticleState(Kill Particles When Lifetime Has Elapsed=false, Let Infinitely Lived Particles Die When Emitter Deactivates=true)
   &SkeletalMeshLocation(
    Bone / Socket Index=$Particles.PartitionParticles.Partition
    MaskSpawnGroup=false
    Mesh Sampling Type=::NewEnumerator5
    Sampled Position Offset=<25 0 0>
    Skeletal Mesh=$System.SkeletalMesh
    Socket Sampling Mode=::NewEnumerator5
   ) {enabled=false}
  ]
  rendererClasses [&NiagaraMeshRendererProperties]
  renderers [{rendererClass=&NiagaraMeshRendererProperties, rendererIndex=0}]
 }
}
summary {
 systemName=NS_Moonflax_Vehicle_LensFlare
 emitters [
  {emitterName=Front_LensFlare, simTarget=GPUComputeSim, rendererClasses [&NiagaraSpriteRendererProperties]}
  {bEnabled=false, emitterName=Front_LensFlare001, simTarget=GPUComputeSim, rendererClasses [&NiagaraSpriteRendererProperties]}
  {bEnabled=false, emitterName=Front_LensFlare002, simTarget=GPUComputeSim, rendererClasses [&NiagaraMeshRendererProperties]}
  {bEnabled=false, emitterName=Front_LensFlare003, simTarget=GPUComputeSim, rendererClasses [&NiagaraMeshRendererProperties]}
 ]
 userVariables [
  User.FlareScale {type=NiagaraFloat, defaultValue=1, description=""}
  User.IsDrifting {type=NiagaraBool, defaultValue=true}
  User.SedanColor {type=NiagaraInt32}
  User.VehicleHasDriver {type=NiagaraBool, defaultValue=true}
 ]
}
